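# MFW_AnimatedGrid is a material function wrapper around MF_AnimatedGrid for an animated cosmetic grid effect.
# It feeds a UV input, defaulting to texture coordinates, plus grouped "Animated Grid" parameters for tiling, two colors, speed with randomness, and line thickness, hardness and color into MF_AnimatedGrid.
# It passes through the color result along with grayscale, random grid, grid line, stepped UV and per-tile UV outputs.
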MaterialFunction /Game/Global/MaterialFunctions/Cosmetics/AnimatedGrid/MFW_AnimatedGrid
graph {nodes [
  n1 = VectorParameter() {group=Animated Grid, parameterName=AnimatedGrid.Color_01, sortPriority=10, defaultValue=<rgba 0.05 0.25 0.35 1>}
  n2 = VectorParameter() {group=Animated Grid, parameterName=AnimatedGrid.Color_02, sortPriority=11, defaultValue=<rgba 0 0.65 1 1>}
  n3 = MaterialFunctionCall!(
   UV=FunctionInput(Preview=TextureCoordinate()) {bUsePreviewValueAsDefault=true, inputName=UV, inputType=FunctionInput_Vector2}
   Tiling=ScalarParameter() {defaultValue=20, group=Animated Grid, parameterName=AnimatedGrid.Tiling, sortPriority=0}
   Tiling X=ScalarParameter() {defaultValue=1, group=Animated Grid, parameterName=AnimatedGrid.Tiling_X, sortPriority=1}
   Tiling Y=ScalarParameter() {defaultValue=1, group=Animated Grid, parameterName=AnimatedGrid.Tiling_Y, sortPriority=2}
   Color 01=@n1.RGB
   Color 02=@n2.RGB
   Speed=ScalarParameter() {defaultValue=0.2, group=Animated Grid, parameterName=AnimatedGrid.Speed, sortPriority=20}
   Speed Random Amount=ScalarParameter() {defaultValue=2, group=Animated Grid, parameterName=AnimatedGrid.Speed_RandomAmount, sortPriority=21}
   Line Thickness=ScalarParameter() {defaultValue=0.1, group=Animated Grid, parameterName=AnimatedGrid.Line_Thickness, sortPriority=30}
   Line Hardness=ScalarParameter() {defaultValue=0.1, group=Animated Grid, parameterName=AnimatedGrid.Line_Hardness, sortPriority=31}
   Line Color=VectorParameter() {group=Animated Grid, parameterName=AnimatedGrid.LineColor, defaultValue=<rgba 0 3 6 1>}.RGB
  ) {materialFunction=&/MF_AnimatedGrid}
  FunctionOutput(None=@n3."Result (V3)") {outputName="Result (Color)"}
  FunctionOutput(None=@n3.Random Grid) {outputName=Random Grid, sortPriority=2}
  FunctionOutput(None=@n3.Gird Lines) {outputName=Grid Lines, sortPriority=3}
  FunctionOutput(None=@n3.Stepped UV) {outputName=Stepped UV, sortPriority=4}
  FunctionOutput(None=@n3.Grayscale Grid) {outputName=Grayscale Grid, sortPriority=1}
  FunctionOutput(None=@n3.Per Tile UV) {outputName=Per Tile UV, sortPriority=5}
 ]}
