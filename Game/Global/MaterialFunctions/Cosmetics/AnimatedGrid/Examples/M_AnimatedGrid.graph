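# M_AnimatedGrid is an example Default Lit material for the cosmetic animated grid material functions, enabled for skeletal meshes and Mesh Deformer.
# It holds two function calls: MFW_AnimatedGrid, which takes only a UV input, and MF_AnimatedGrid, which exposes tiling, two colors, speed with randomness, and line styling inputs.
# The graph shows no connections between the calls or into the material outputs.
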Material /Game/Global/MaterialFunctions/Cosmetics/AnimatedGrid/Examples/M_AnimatedGrid
graph {
 properties {bAutomaticallySetUsageInEditor=false, bUsedWithMeshDeformer=true, bUsedWithSkeletalMesh=true, usedShadingModels=Default Lit}
 nodes [
  MaterialFunctionCall!(UV) {materialFunction=/Game/Global/MaterialFunctions/Cosmetics/AnimatedGrid/MFW_AnimatedGrid}
  MaterialFunctionCall!(UV, Tiling, Tiling X, Tiling Y, Color 01, Color 02, Speed, Speed Random Amount, Line Thickness, Line Hardness, Line Color) {materialFunction=/Game/Global/MaterialFunctions/Cosmetics/AnimatedGrid/MF_AnimatedGrid}
 ]
}
