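# Material function wrapper that exposes MF_AnimatedGridTexture as grouped material parameters for an animated, colored grid effect in cosmetic materials.
# UVs scaled by uniform and per-axis tiling sample a grid mask texture, with a selectable channel feeding the grid input.
# A static switch takes grid lines from a separate texture or the same mask, then applies channel selection and an amount multiplier.
# Color, line color and speed parameters feed the inner function, whose color and grayscale results are output.
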MaterialFunction /Game/Global/MaterialFunctions/Cosmetics/AnimatedGrid/MFW_AnimatedGridTexture
graph {expressions {
  MaterialExpressionChannelMaskParameter_0 {
   class=MaterialExpressionChannelMaskParameter
   inputs [None {from=MaterialExpressionTextureSampleParameter2D_0, output=RGB}]
   properties {group=Animated Grid Texture, maskChannel=Green, parameterName=AnimatedGridTexture.Grid_Channel, sortPriority=10, defaultValue=<rgba 0 1 0 0>}
  }
  MaterialExpressionChannelMaskParameter_1 {
   class=MaterialExpressionChannelMaskParameter
   inputs [None {from=MaterialExpressionStaticSwitchParameter_0, output=""}]
   properties {group=Animated Grid Texture, parameterName=AnimatedGridTexture.GridLines_Channel, sortPriority=11}
  }
  MaterialExpressionConvert_0 {
   class=MaterialExpressionConvert
   inputs [
    None {from=MaterialExpressionScalarParameter_1}
    None {from=MaterialExpressionScalarParameter_2}
   ]
   properties {
    nodeName=Make Vector2
    convertInputs [
     {
      type=Scalar
      expressionInput {
       expression=&self:MaterialExpressionScalarParameter_1
       inputName=None
       mask=0
       maskA=0
       maskB=0
       maskG=0
       maskR=0
       outputIndex=0
      }
     }
     {
      type=Scalar
      expressionInput {
       expression=&self:MaterialExpressionScalarParameter_2
       inputName=None
       mask=0
       maskA=0
       maskB=0
       maskG=0
       maskR=0
       outputIndex=0
      }
     }
    ]
    convertOutputs [{type=Vector2}]
   }
  }
  MaterialExpressionFunctionOutput_0 {
   class=MaterialExpressionFunctionOutput
   inputs [None {from=MaterialExpressionMaterialFunctionCall_0, output="Result (V3)"}]
   properties {outputName="Result (V3)"}
  }
  MaterialExpressionFunctionOutput_1 {
   class=MaterialExpressionFunctionOutput
   inputs [None {from=MaterialExpressionMaterialFunctionCall_0, output=Grayscale Grid}]
   properties {outputName=Grayscale Grid, sortPriority=1}
  }
  MaterialExpressionMaterialFunctionCall_0 {
   class=MaterialExpressionMaterialFunctionCall
   inputs [
    Grid {from=MaterialExpressionNamedRerouteUsage_0}
    Grid Lines {from=MaterialExpressionNamedRerouteUsage_1}
    Color 01 {from=MaterialExpressionVectorParameter_0, output=RGB}
    Color 02 {from=MaterialExpressionVectorParameter_1, output=RGB}
    Speed {from=MaterialExpressionScalarParameter_3}
    Speed Random Amount {from=MaterialExpressionScalarParameter_4}
    Line Color {from=MaterialExpressionVectorParameter_2, output=RGB}
   ]
   properties {materialFunction=&/MF_AnimatedGridTexture}
  }
  MaterialExpressionMultiply_0 {
   class=MaterialExpressionMultiply
   inputs [
    A {from=MaterialExpressionChannelMaskParameter_1, output=R}
    B {from=MaterialExpressionScalarParameter_0}
   ]
  }
  MaterialExpressionMultiply_1 {
   class=MaterialExpressionMultiply
   inputs [
    A {from=MaterialExpressionTextureCoordinate_0}
    B {from=MaterialExpressionMultiply_5}
   ]
  }
  MaterialExpressionMultiply_5 {
   class=MaterialExpressionMultiply
   inputs [
    A {from=MaterialExpressionScalarParameter_5}
    B {from=MaterialExpressionConvert_0}
   ]
  }
  MaterialExpressionNamedRerouteDeclaration_0 {
   class=MaterialExpressionNamedRerouteDeclaration
   inputs [None {from=MaterialExpressionMultiply_0}]
   properties {name=GridLines, nodeColor=<rgba 1 0.1172 0 1>}
  }
  MaterialExpressionNamedRerouteDeclaration_1 {
   class=MaterialExpressionNamedRerouteDeclaration
   inputs [None {from=MaterialExpressionChannelMaskParameter_0, output=R}]
   properties {name=Grid, nodeColor=<rgba 0 0.2031 1 1>}
  }
  MaterialExpressionNamedRerouteUsage_0 {class=MaterialExpressionNamedRerouteUsage}
  MaterialExpressionNamedRerouteUsage_1 {class=MaterialExpressionNamedRerouteUsage}
  MaterialExpressionScalarParameter_0 {
   class=MaterialExpressionScalarParameter
   properties {defaultValue=1, group=Animated Grid Texture, parameterName=AnimatedGridTexture.GridLines_Amount, sortPriority=12}
  }
  MaterialExpressionScalarParameter_1 {
   class=MaterialExpressionScalarParameter
   properties {defaultValue=1, group=Animated Grid Texture, parameterName=AnimatedGridTexture.Tiling_X, sortPriority=16}
  }
  MaterialExpressionScalarParameter_2 {
   class=MaterialExpressionScalarParameter
   properties {defaultValue=1, group=Animated Grid Texture, parameterName=AnimatedGridTexture.Tiling_Y, sortPriority=17}
  }
  MaterialExpressionScalarParameter_3 {
   class=MaterialExpressionScalarParameter
   properties {defaultValue=0.2, group=Animated Grid Texture, parameterName=AnimatedGridTexture.Speed, sortPriority=30}
  }
  MaterialExpressionScalarParameter_4 {
   class=MaterialExpressionScalarParameter
   properties {defaultValue=2, group=Animated Grid Texture, parameterName=AnimatedGridTexture.Speed_RandomAmount, sortPriority=31}
  }
  MaterialExpressionScalarParameter_5 {
   class=MaterialExpressionScalarParameter
   properties {defaultValue=1, group=Animated Grid Texture, parameterName=AnimatedGridTexture.Tiling, sortPriority=15}
  }
  MaterialExpressionStaticSwitchParameter_0 {
   class=MaterialExpressionStaticSwitchParameter
   inputs [
    True {from=MaterialExpressionTextureSampleParameter2D_1, output=RGB}
    False {from=MaterialExpressionTextureSampleParameter2D_0, output=RGB}
   ]
   properties {group=Animated Grid Texture, parameterName=AnimatedGridTexture.UseGridLinesTexture, sortPriority=2}
  }
  MaterialExpressionTextureCoordinate_0 {class=MaterialExpressionTextureCoordinate}
  MaterialExpressionTextureSampleParameter2D_0 {
   class=MaterialExpressionTextureSampleParameter2D
   inputs [
    UVs {from=MaterialExpressionMultiply_1, output=""}
    Apply View MipBias {from="", output=""}
   ]
   properties {
    group=Animated Grid Texture
    parameterName=AnimatedGridTexture.GridMask_Texture
    samplerSource=SSM_Wrap_WorldGroupSettings
    samplerType=SAMPLERTYPE_Masks
    sortPriority=0
    texture=&T_WhiteMask
   }
  }
  MaterialExpressionTextureSampleParameter2D_1 {
   class=MaterialExpressionTextureSampleParameter2D
   inputs [
    UVs {from=MaterialExpressionMultiply_1, output=""}
    Apply View MipBias {from="", output=""}
   ]
   properties {
    group=Animated Grid Texture
    parameterName=AnimatedGridTexture.GridLines_Texture
    samplerSource=SSM_Wrap_WorldGroupSettings
    samplerType=SAMPLERTYPE_Masks
    sortPriority=3
    texture=&T_BlackMask
   }
  }
  MaterialExpressionVectorParameter_0 {
   class=MaterialExpressionVectorParameter
   properties {group=Animated Grid Texture, parameterName=AnimatedGridTexture.Color_01, sortPriority=20, defaultValue=<rgba 0.05 0.25 0.35 1>}
  }
  MaterialExpressionVectorParameter_1 {
   class=MaterialExpressionVectorParameter
   properties {group=Animated Grid Texture, parameterName=AnimatedGridTexture.Color_02, sortPriority=21, defaultValue=<rgba 0 0.65 1 1>}
  }
  MaterialExpressionVectorParameter_2 {
   class=MaterialExpressionVectorParameter
   properties {group=Animated Grid Texture, parameterName=AnimatedGridTexture.LineColor, sortPriority=40, defaultValue=<rgba 0 3 6 1>}
  }
 }}
