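# Example material for the animated grid texture cosmetic material function, using the Default Lit shading model.
# Its graph holds a single call to MFW_AnimatedGridTexture, with no other logic or parameters of its own.
# It is flagged for skeletal meshes and mesh deformers, with automatic usage setting in the editor turned off.
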Material /Game/Global/MaterialFunctions/Cosmetics/AnimatedGrid/Examples/M_AnimatedGridTexture
graph {
 properties {bAutomaticallySetUsageInEditor=false, bUsedWithMeshDeformer=true, bUsedWithSkeletalMesh=true, usedShadingModels=Default Lit}
 nodes [MaterialFunctionCall() {materialFunction=/Game/Global/MaterialFunctions/Cosmetics/AnimatedGrid/MFW_AnimatedGridTexture}]
}
